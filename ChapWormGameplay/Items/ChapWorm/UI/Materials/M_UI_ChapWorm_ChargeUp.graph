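# Translucent UI material that draws the ChapWorm item's charge-up meter as a purple fill inside a bracket shape.
# Progress lerps between ProgressStart and ProgressEnd, sliding a rotated-UV mask with a glowing leading edge.
# Named reroutes combine bracket and fill textures with noise, an animated octagon pattern and shine through color-dodge blends.
# Scalar lerp switches select charged idle, boosted charge and supercharge depletion looks, the last using radially panned noise.
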Material /ChapWormGameplay/Items/ChapWorm/UI/Materials/M_UI_ChapWorm_ChargeUp
graph {
 properties {blendMode=BLEND_Translucent, materialDomain=MD_UI, pixelDepthOffsetMode=PDOM_Legacy, usedShadingModels=Default Lit}
 nodes [
  Add()
  Constant2Vector() {g=-0.1, r=-0.2}
  Constant3Vector() {constant=<rgba 0.2367 0.0077 1 1>}
  Constant() {desc="When progress value changes, it defines where change ends.\r\n0 - 1", r=0.81}
  n1 = Multiply(A=Multiply(A=Time()) {constB=-1}) {constB=0.3}
  Frac(None=Add(A=Frac(None=@n1)))
  n2 = NamedRerouteUsage()
  n3 = Multiply(A=@n2, B=NamedRerouteUsage())
  n4 = Add(A=@n2, B=@n3)
  n5 = NamedRerouteUsage()
  n6 = Add(A=NamedRerouteUsage(), B=LinearInterpolate(A=@n4, B=Add(A=@n4, B=@n5)) {constAlpha=0})
  LinearInterpolate(A=@n6, B=Add(A=@n6, B=NamedRerouteUsage()), Alpha=ScalarParameter() {parameterName=SCDepleteGlow_Switch})
  n7 = ScalarParameter() {
   defaultValue=1.1
   desc="Property for animating progress value change (from ProgressStart to ProgressEnd)\r\n0 - 1"
   group=Progress
   parameterName=Progress
   sliderMax=1
   sortPriority=3
  }
  LinearInterpolate(Alpha=@n7) {constA=0.65, constB=0.75}
  MaterialFunctionCall(Speed=Constant() {r=0.2}) {materialFunction=&DebugTimeSine}
  n8 = TextureCoordinate()
  n9 = Constant() {r=0.33}
  n10 = ScalarParameter() {
   defaultValue=0.2284
   desc="When progress value changes, it defines where change starts.\r\n0 - 1"
   group=Progress
   parameterName=ProgressStart
   sliderMax=1
   sortPriority=1
  }
  n11 = ScalarParameter() {
   defaultValue=0.743
   desc="When progress value changes, it defines where change ends.\r\n0 - 1"
   group=Progress
   parameterName=ProgressEnd
   sliderMax=1
   sortPriority=2
  }
  n12 = ScalarParameter() {
   defaultValue=1.1
   desc="Property for animating progress value change (from ProgressStart to ProgressEnd)\r\n0 - 1"
   group=Progress
   parameterName=Progress
   sliderMax=1
   sortPriority=3
  }
  n13 = Add(
   A=Multiply(A=MaterialFunctionCall(UVs=@n8, "Rotation Angle (0-1)"=@n9) {materialFunction=&CustomRotator}.Rotated Values) {constB=-1}
   B=NamedRerouteDeclaration(None=LinearInterpolate(A=@n10, B=@n11, Alpha=@n12)) {name=Progress, nodeColor=<rgba 0 0 0 1>}
  ) {constB=0.5}
  n14 = SmoothStep(
   Max=ScalarParameter() {defaultValue=0.01, group=Progress, parameterName=ProgressEdgeSharpness, sliderMax=1, sortPriority=2}
   Value=ComponentMask(None=@n13) {g=true}
  ) {constMax=0.01}
  n15 = OneMinus(None=ComponentMask(None=@n13) {g=true})
  n16 = Multiply(A=@n14, B=Subtract(A=@n15) {constB=0.85})
  NamedRerouteDeclaration(None=Saturate(None=Multiply(A=@n16) {constB=5})) {name=ProgressTopGlow, nodeColor=<rgba 0 0 0 1>}
  n17 = NamedRerouteUsage()
  n18 = Constant2Vector() {g=0.015, r=0.02}
  n19 = TextureSample(UVs=MaterialFunctionCall(UVs=Add(A=@n17, B=@n18) {constB=0}, Texture Scale=ScalarParameter() {defaultValue=0.68, parameterName=ProgressBarSize}) {materialFunction=&ScaleUVsByCenter}.UVs) {samplerType=SAMPLERTYPE_LinearGrayscale, texture=/ChapWormGameplay/Items/ChapWorm/UI/Textures/T_UI_ChapWorm_ChargeFill}
  NamedRerouteDeclaration(None=SmoothStep(Value=@n19.R)) {name=FillGlow, nodeColor=<rgba 0 0 0 1>}
  NamedRerouteDeclaration(None=LinearInterpolate(
    A=VectorParameter() {group=Color, parameterName=Shadow, defaultValue=<rgba 0 0 0 1>}.RGB
    B=VectorParameter() {group=Color, parameterName=Bracket, defaultValue=<rgba 0.7526 0.7526 0.7526 1>}.RGB
    Alpha=NamedRerouteUsage()
   )) {name=Color:Bracket, nodeColor=<rgba 0 0 0 1>}
  n20 = MaterialFunctionCall() {materialFunction=&LinearGradient}
  n21 = Multiply(A=OneMinus(None=Add(A=@n20.VGradient) {constB=0.2})) {constB=3}
  n22 = MaterialFunctionCall(UVs=Add(A=NamedRerouteUsage(), B=Constant2Vector() {g=0.015, r=0.02}) {constB=0}, Texture Scale=Constant() {r=0.68}) {materialFunction=&ScaleUVsByCenter}
  n23 = Add(
   A=SmoothStep(Value=TextureSample(UVs=@n22.UVs) {samplerType=SAMPLERTYPE_LinearGrayscale, texture=&T_UI_ChapWorm_ChargeFill}.R) {constMin=0.3}
   B=NamedRerouteUsage()
  )
  n24 = Multiply(A=@n23, B=NamedRerouteUsage())
  n25 = Multiply(A=@n23, B=NamedRerouteUsage())
  NamedRerouteDeclaration(None=Add(A=Multiply(A=@n21, B=@n24), B=@n25)) {name=SCDepleteGlow, nodeColor=<rgba 0 0 0 1>}
  n26 = Rotator(Coordinate=TextureCoordinate(), Time=Constant() {r=3})
  NamedRerouteDeclaration(None=Subtract(A=MaterialFunctionCall(UV Channel=@n26) {materialFunction=&LinearGradient}.UGradient) {constB=0.45}) {name=LinearGradient, nodeColor=<rgba 0 0 0 1>}
  NamedRerouteDeclaration(None=Saturate(None=@n14)) {name=ProgressGradient, nodeColor=<rgba 0 0 0 1>}
  n27 = Multiply(A=TextureCoordinate() {uTiling=1.6, vTiling=1.6}, B=AppendVector(A=Constant() {r=1.3}, B=Constant() {r=0.8}))
  n28 = Multiply(A=AppendVector(A=Constant() {r=0.05}, B=Constant() {r=0.07}), B=Constant() {r=10})
  n29 = Multiply(A=TextureSample(UVs=Panner(Coordinate=@n27, Speed=@n28) {speedX=0.05, speedY=-0.125}) {samplerType=SAMPLERTYPE_Grayscale, texture=&LowResBlurredNoise}.R) {constB=1.5}
  NamedRerouteDeclaration(None=Saturate(None=Power(Base=@n29))) {name=Noise, nodeColor=<rgba 0 0 0 1>}
  NamedRerouteDeclaration(None=@n3) {name=MaskedProgressGradient, nodeColor=<rgba 0 0 0 1>}
  n30 = NamedRerouteUsage()
  n31 = NamedRerouteUsage()
  n32 = Constant3Vector() {constant=<rgba 0.2303 0 1 1>}
  n33 = Constant3Vector() {constant=<rgba 0.2999 0.0002 0.7396 1>}
  n34 = MaterialFunctionCall(A=@n31, B=@n32, C=@n33, Alpha=NamedRerouteUsage()) {materialFunction=&3ColorBlend}
  n35 = MaterialFunctionCall!(Blend=NamedRerouteUsage(), Base=@n34.Result) {materialFunction=/Engine/Functions/Engine_MaterialFunctions03/Blends/Blend_ColorDodge}
  NamedRerouteDeclaration(None=MaterialFunctionCall(
    Blend=NamedRerouteUsage()
    Base=MaterialFunctionCall(Blend=@n30, Base=@n35.Result) {materialFunction=&Blend_ColorDodge}.Result
   ) {materialFunction=&Blend_ColorDodge}.Result) {name=Color:BoostedCharge, nodeColor=<rgba 0 0 0 1>}
  n36 = Multiply(A=NamedRerouteUsage()) {constB=0.2}
  n37 = LinearInterpolate(A=Constant3Vector() {constant=<rgba 0.219 0 0.9401 1>}, B=Constant3Vector() {constant=<rgba 0.5091 0 1 1>}, Alpha=NamedRerouteUsage())
  NamedRerouteDeclaration(None=Saturate(None=MaterialFunctionCall(Blend=@n36, Base=@n37) {materialFunction=&Blend_ColorDodge}.Result)) {name=Color:ProgressLinearGradient, nodeColor=<rgba 0 0 0 1>}
  n38 = LinearInterpolate(
   A=Constant3Vector() {constant=<rgba 0 0 0 1>}
   B=VectorParameter() {group=Color, parameterName=FillBg, defaultValue=<rgba 0.0117 0.0045 0.0573 1>}.RGB
   Alpha=NamedRerouteUsage()
  )
  n39 = Multiply(A=NamedRerouteUsage()) {constB=0.5}
  n40 = LinearInterpolate(A=@n38, B=NamedRerouteUsage(), Alpha=NamedRerouteUsage())
  n41 = LinearInterpolate(
   A=LinearInterpolate(
    A=@n38
    B=MaterialFunctionCall(Blend=@n39, Base=@n40) {materialFunction=&Blend_ColorDodge}.Result
    Alpha=ScalarParameter() {parameterName=Charged_IdleFX}
   )
   B=LinearInterpolate(A=NamedRerouteUsage(), B=NamedRerouteUsage(), Alpha=ScalarParameter() {parameterName=BoostedChargeFX_Switch})
   Alpha=NamedRerouteUsage()
  )
  n42 = NamedRerouteUsage()
  n43 = ScalarParameter() {parameterName=SCDepleteGlow_Switch}
  NamedRerouteDeclaration(None=LinearInterpolate(A=LinearInterpolate(A=@n41, B=@n42, Alpha=@n43), B=Constant3Vector() {constant=<rgba 0.1668 0.0559 0.5365 1>}, Alpha=NamedRerouteUsage())) {name=FinalColor, nodeColor=<rgba 0 0 0 1>}
  n44 = MaterialFunctionCall!(
   X Scale=ScalarParameter() {defaultValue=4, parameterName=X Scale}
   Time
   Y Scale=ScalarParameter() {defaultValue=4, parameterName=Y Scale}
   Flip UVs
   X Speed=ScalarParameter() {defaultValue=-0.2, parameterName=X Speed}
   Y Speed=ScalarParameter() {defaultValue=0.03, parameterName=Y Speed}
   UV
  ) {materialFunction=&/MF_RadialPanner_ChapWorm}
  NamedRerouteDeclaration(None=@n44.Result) {name=RadialPannerUVs, nodeColor=<rgba 0 1 0.7188 1>}
  NamedRerouteDeclaration(None=MaterialFunctionCall() {materialFunction=&GetUserInterfaceUV}.Normalized UV) {name=Normalized_UVs, nodeColor=<rgba 0 0 0 1>}
  n45 = TextureCoordinate() {uTiling=2, vTiling=2}
  n46 = AppendVector(
   A=ScalarParameter() {defaultValue=-0.15, parameterName=Embers02_XSpeed}
   B=ScalarParameter() {defaultValue=0.3, parameterName=Embers02_YSpeed}
  )
  n47 = Multiply(
   A=TextureSample(UVs=@n44.Result) {texture=/ChapWormGameplay/Items/ChapWorm/UI/Textures/T_UI_ChapWorm_Reticle_VFX_02}.R
   B=TextureSample(UVs=Panner(Coordinate=@n45, Speed=@n46) {speedY=0.2}) {texture=&T_UI_ChapWorm_Reticle_VFX_02}.R
  )
  NamedRerouteDeclaration(None=Multiply(A=Saturate(None=@n47)) {constB=0.9}) {name=SCDeplete_Noise, nodeColor=<rgba 1 0.4688 0 1>}
  n48 = NamedRerouteUsage()
  n49 = MaterialFunctionCall(A=@n31, B=@n32, C=@n33, Alpha=@n48) {materialFunction=&3ColorBlend}
  NamedRerouteDeclaration(None=MaterialFunctionCall(
    Blend=@n30
    Base=MaterialFunctionCall(Blend=@n48, Base=@n49.Result) {materialFunction=&Blend_ColorDodge}.Result
   ) {materialFunction=&Blend_ColorDodge}.Result) {name=Color:SCDeplete, nodeColor=<rgba 0 0 0 1>}
  n50 = Multiply(A=@n14, B=Subtract(A=@n15) {constB=0.97})
  NamedRerouteDeclaration(None=Saturate(None=Multiply(A=@n50) {constB=40})) {name=ProgressTopGlowShine, nodeColor=<rgba 0 0 0 1>}
  n51 = MaterialFunctionCall() {materialFunction=&GetUserInterfaceUV}
  NamedRerouteDeclaration(None=@n51.Normalized UV) {name=Normalized UV, nodeColor=<rgba 0 0 0 1>}
  NamedRerouteDeclaration(None=@n25) {name=SDCDepleteTopGlow, nodeColor=<rgba 0 0 0 1>}
  n52 = Frac(None=Multiply(A=Time(), B=Constant() {r=1}))
  n53 = Clamp(None=MaterialFunctionCall!(UVs=NamedRerouteUsage(), Offset=Multiply(A=@n52) {constB=4}, RotationCenter=Constant() {r=0.8}, RotationAngle=Constant() {r=0.2}) {materialFunction=/FortUILibrary/MaterialFunctions/Gradients/MF_UI_ReflectedGradient}.Result) {maxDefault=0.9, minDefault=0.1}
  NamedRerouteDeclaration(None=SmoothStep(Value=Power(Base=@n53)) {constMax=0.8, constMin=0.2}) {name=Shine, nodeColor=<rgba 0 0 0 1>}
  NamedRerouteDeclaration(None=@n23) {name=ChunkyMunkyMask, nodeColor=<rgba 0 0 0 1>}
  n54 = NamedRerouteUsage()
  n55 = MaterialFunctionCall(Blend=Multiply(A=NamedRerouteUsage()) {constB=0.8}, Base=@n34.Result) {materialFunction=&Blend_ColorDodge}
  NamedRerouteDeclaration(None=MaterialFunctionCall(Blend=Multiply(A=@n54) {constB=0.99}, Base=Saturate(None=@n55.Result)) {materialFunction=&Blend_ColorDodge}.Result) {name=Charged_IdleFX, nodeColor=<rgba 0 0 0 1>}
  n56 = NamedRerouteUsage()
  n57 = Constant3Vector() {constant=<rgba 0.23 0.1756 0.6198 1>}
  n58 = LinearInterpolate(
   A=@n56
   B=Add(
    A=MaterialFunctionCall(Blend=@n57, Base=@n56) {materialFunction=&Blend_ColorDodge}.Result
    B=NamedRerouteUsage()
   )
   Alpha=NamedRerouteUsage()
  )
  n59 = Constant3Vector() {constant=<rgba 0.8499 0.0905 1 1>}
  NamedRerouteDeclaration(None=LinearInterpolate(
    A=@n58
    B=MaterialFunctionCall!(Blend=@n59, Base=@n58) {materialFunction=/Engine/Functions/Engine_MaterialFunctions03/Blends/Blend_LinearDodge}.Result
    Alpha=NamedRerouteUsage()
   )) {name=Final: SuperChargedDeplete, nodeColor=<rgba 0.1016 0 1 1>}
  NamedRerouteDeclaration(None=@n41) {name=FoundationColor, nodeColor=<rgba 0.3771 0 1 1>}
  n60 = Subtract(A=@n51.Normalized UV, B=Constant2Vector()) {constB=0.1}
  n61 = Constant() {r=0.4}
  n62 = TextureSample(UVs=Rotator(
    Coordinate=MaterialFunctionCall(UVs=@n60, Texture Scale=@n61) {materialFunction=&ScaleUVsByCenter}.UVs
    Time=Constant() {r=3.2575}
   )) {texture=/ChapWormGameplay/Items/ChapWorm/UI/Textures/T_UI_ChapWorm_OctagonPattern}
  n63 = Time()
  n64 = Multiply(A=Add(A=@n62.G, B=Multiply(A=@n63) {constB=0.8})) {constB=2}
  NamedRerouteDeclaration(None=SmoothStep(Value=Frac(None=@n64)) {constMax=0.2, constMin=0.8}) {name=Shine Path, nodeColor=<rgba 0 0 0 1>}
  n65 = Multiply(A=NamedRerouteUsage()) {constB=50}
  n66 = Multiply(A=NamedRerouteUsage()) {constB=50}
  NamedRerouteDeclaration(None=Subtract(A=Saturate(None=@n65), B=Saturate(None=@n66))) {name=Outline, nodeColor=<rgba 0 0 0 1>}
  n67 = NamedRerouteUsage()
  n68 = Constant2Vector() {g=-0.07, r=-0.07}
  n69 = TextureSample(UVs=MaterialFunctionCall(UVs=Add(A=@n67, B=@n68) {constB=0}, Texture Scale=Constant() {r=0.8}) {materialFunction=&ScaleUVsByCenter}.UVs) {samplerType=SAMPLERTYPE_LinearGrayscale, texture=/ChapWormGameplay/Items/ChapWorm/UI/Textures/T_UI_ChapWorm_ChargeUp}
  NamedRerouteDeclaration(None=Multiply(A=SmoothStep(Value=@n69.R) {constMax=0.75, constMin=0.7}) {constB=0}) {name=Shadow_Bracket, nodeColor=<rgba 0 0 0 1>}
  n70 = Multiply(A=SmoothStep(Value=@n19.R) {constMax=0.75, constMin=0.7}) {constB=0.8}
  NamedRerouteDeclaration(None=@n70) {name=FillBg, nodeColor=<rgba 0 0 0 1>}
  n71 = NamedRerouteUsage()
  n72 = TextureSample(UVs=MaterialFunctionCall(UVs=Add(A=@n71) {constB=0}, Texture Scale=Constant() {r=0.8}) {materialFunction=&ScaleUVsByCenter}.UVs) {samplerType=SAMPLERTYPE_LinearGrayscale, texture=&T_UI_ChapWorm_ChargeUp}
  n73 = SmoothStep(Value=@n72.R) {constMax=0.75, constMin=0.7}
  NamedRerouteDeclaration(None=@n73) {name=Bracket, nodeColor=<rgba 0 0 0 1>}
  n74 = Multiply(A=SmoothStep(Value=@n72.R) {constMax=0.7, constMin=0.6}) {constB=0.7}
  NamedRerouteDeclaration(None=Multiply(A=Add(A=@n73, B=@n74)) {constB=0.5}) {name=BracketShadow, nodeColor=<rgba 0 0 0 1>}
  n75 = Multiply(A=SmoothStep(Value=@n19.R) {constMax=0.75, constMin=0.6}) {constB=0.25}
  NamedRerouteDeclaration(None=Multiply(A=Add(A=@n70, B=@n75)) {constB=0.7}) {name=FillBgShadow, nodeColor=<rgba 0 0 0 1>}
  n76 = SmoothStep(Value=@n62.R) {constMax=0.7, constMin=0.4}
  n77 = Multiply(A=@n76) {constB=0.1}
  n78 = Multiply(A=NamedRerouteUsage()) {constB=0.5}
  NamedRerouteDeclaration(None=Multiply(A=@n76, B=Add(A=@n77, B=@n78))) {name=Pattern_Mask, nodeColor=<rgba 0 0 0 1>}
  NamedRerouteUsage()
 ]
}
