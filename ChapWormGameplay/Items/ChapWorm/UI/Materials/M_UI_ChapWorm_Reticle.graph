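# Translucent UI material that draws the ChapWorm item's reticle, with a center dot, three pips and an in-range state.
# Smoothstep thresholds carve shape and glow from two grayscale textures, blended by InRange_Switch, which also applies the HUD collection's target-in-range color.
# Procedural SDF boxes form a top pip and mirrored, rotated bottom pips with shadows.
# Scalar parameters drive pip lock-in position, stretch, ready color and flashing, including a time-sine pulse.
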Material /ChapWormGameplay/Items/ChapWorm/UI/Materials/M_UI_ChapWorm_Reticle
graph {
 properties {bAutomaticallySetUsageInEditor=false, blendMode=BLEND_Translucent, materialDomain=MD_UI, usedShadingModels=Default Lit}
 nodes [
  n1 = MaterialFunctionCall(Texture Scale=ScalarParameter() {defaultValue=0.9, group=Scale, parameterName=ReticleScale}) {materialFunction=&ScaleUVsByCenter}
  n2 = TextureSample(UVs=@n1.UVs) {
   samplerSource=SSM_Clamp_WorldGroupSettings
   samplerType=SAMPLERTYPE_LinearGrayscale
   texture=/ChapWormGameplay/Items/ChapWorm/UI/Textures/T_UI_ChapWorm_Reticle
  }
  n3 = SmoothStep(Value=@n2.R) {constMax=0.8, constMin=0.75}
  n4 = SmoothStep(Value=@n2.R) {constMax=0.75, constMin=0.65}
  n5 = Multiply(A=@n4) {constB=0.5}
  n6 = TextureSample(UVs=@n1.UVs) {
   samplerSource=SSM_Clamp_WorldGroupSettings
   samplerType=SAMPLERTYPE_LinearGrayscale
   texture=/ChapWormGameplay/Items/ChapWorm/UI/Textures/T_UI_ChapWorm_Reticle_InRange
  }
  n7 = SmoothStep(Value=@n6.R) {constMax=0.8, constMin=0.75}
  n8 = SmoothStep(Value=@n6.R) {constMax=0.75, constMin=0.65}
  n9 = LinearInterpolate(
   A=Multiply(A=Add(A=@n3, B=@n5)) {constB=0.6}
   B=Add(A=Multiply(A=@n7) {constB=0.5}, B=Multiply(A=@n8) {constB=0.4})
   Alpha=ScalarParameter() {parameterName=InRange_Switch}
  )
  n10 = Multiply(A=NamedRerouteUsage()) {constB=0.4}
  Add(A=Add(A=NamedRerouteUsage(), B=Add(A=@n9, B=@n10)), B=NamedRerouteUsage())
  AppendVector(A=Constant() {r=-0.08}, B=Constant() {r=-0.09})
  AppendVector(A=Constant() {r=0.232}, B=Constant() {r=-0.2})
  n11 = LinearInterpolate(
   A=Constant3Vector()
   B=Constant3Vector() {constant=<rgba 1 1 1 1>}
   Alpha=LinearInterpolate(A=@n3, B=@n7, Alpha=ScalarParameter() {parameterName=InRange_Switch})
  )
  n12 = Constant3Vector() {constant=<rgba 1 0.9062 0.985 1>}
  n13 = Constant3Vector() {constant=<rgba 0.3954 0.0906 1 1>}
  n14 = ScalarParameter() {parameterName=PipReadyFlash}
  n15 = LinearInterpolate(
   A=LinearInterpolate(
    A=Constant3Vector() {constant=<rgba 0.1542 0.0411 0.7891 1>}
    B=LinearInterpolate(A=@n12, B=@n13, Alpha=@n14)
    Alpha=ScalarParameter() {parameterName=BottomPip_ReadySwitch}
   )
   B=LinearInterpolate(
    A=Constant3Vector() {constant=<rgba 0.6924 0.5365 1 1>}
    B=Constant3Vector() {constant=<rgba 0.1601 0.0367 0.4036 1>}
    Alpha=MaterialFunctionCall() {materialFunction=&DebugTimeSine}.Result
   )
   Alpha=ScalarParameter() {parameterName=UseMeflash}
  )
  n16 = NamedRerouteUsage()
  n17 = AppendVector(
   A=LinearInterpolate(A=LinearInterpolate(A=@n11, B=@n15, Alpha=@n16), B=Constant3Vector() {constant=<rgba 1 1 1 1>}, Alpha=NamedRerouteUsage())
   B=Constant() {r=1}
  )
  LinearInterpolate(
   A=@n17
   B=Multiply(A=@n17, B=CollectionParameter() {collection=/Game/UI/HUD/Materials/MPC_HUD_Globals, parameterName=V_Color_TargetInRange})
   Alpha=ScalarParameter() {parameterName=InRange_Switch}
  )
  LinearInterpolate() {constA=-0.75}
  LinearInterpolate(A=Constant() {r=0.46}, B=Constant() {r=0.32})
  MaterialFunctionCall() {materialFunction=&DebugTimeSine}
  NamedRerouteDeclaration(None=@n4) {name=Reticle_OuterGlow, nodeColor=<rgba 0 0 0 1>}
  NamedRerouteDeclaration(None=@n8) {name=InRange_OuterGlow, nodeColor=<rgba 0 0 0 1>}
  n18 = MaterialFunctionCall!(Size=ScalarParameter() {defaultValue=0.035, parameterName=CenterDot}) {materialFunction=/Game/Athena/HUD/reticles/MaterialFunctions/MF_Athena_Reticle_CenterDot}
  NamedRerouteDeclaration(None=@n18.Dot) {name=CenterDot_Color, nodeColor=<rgba 0 0 0 1>}
  n19 = TextureCoordinate()
  n20 = ComponentMask(None=@n19) {r=true}
  n21 = Add(
   A=ComponentMask(None=@n19) {g=true}
   B=LinearInterpolate(A=Constant() {r=0.46}, B=Constant() {r=0.32}, Alpha=ScalarParameter() {defaultValue=1, parameterName=Pip_LockInSwitch})
  )
  n22 = ScalarParameter() {parameterName=Pip_StretchSwitch}
  n23 = MaterialFunctionCall(
   UVs=MaterialFunctionCall(UVs=AppendVector(A=@n20, B=@n21), Texture Scale=Constant() {r=0.25}) {materialFunction=&ScaleUVsByCenter}.UVs
   Size=LinearInterpolate(A=Constant2Vector() {g=0.1, r=0.3}, B=Constant2Vector() {g=0.1, r=0.8}, Alpha=LinearInterpolate(Alpha=@n22) {constA=-0.75})
   CornerRoundness=Constant() {r=0.05}
  ) {materialFunction=&MF_UI_Box}
  NamedRerouteDeclaration(None=SmoothStep(Value=@n23.SDF) {constMax=0.02, constMin=0.1}) {name=Top_Pip_DropShadow, nodeColor=<rgba 0 0 0 1>}
  NamedRerouteDeclaration(None=Add(A=NamedRerouteUsage(), B=NamedRerouteUsage())) {name=Pips_Shadows, nodeColor=<rgba 0 0 0 1>}
  NamedRerouteDeclaration(None=SmoothStep(Value=@n23.SDF) {constMax=-0.006, constMin=0.004}) {name=Top_Pip, nodeColor=<rgba 0 0 0 1>}
  n24 = TextureCoordinate()
  n25 = LinearInterpolate(
   A=AppendVector(A=Constant() {r=0.232}, B=Constant() {r=-0.172})
   B=AppendVector(A=Constant() {r=-0.025}, B=Constant() {r=-0.099})
   Alpha=ScalarParameter() {defaultValue=1, parameterName=Pip_LockInSwitch}
  ) {constAlpha=0}
  n26 = Constant2Vector() {g=0.25, r=0.5}
  n27 = ScalarParameter() {parameterName=Pip_StretchSwitch}
  n28 = MaterialFunctionCall(
   UVs=MaterialFunctionCall(
    UVs=MaterialFunctionCall!(UV=@n24, Offset=@n25, Rotation, Scale=@n26, MirrorY) {materialFunction=/FortUILibrary/MaterialFunctions/Basics/MF_UI_Symmetry_2x1_DEPRECATED}.Result
    "Rotation Angle (0-1)"=Constant() {r=0.345}
   ) {materialFunction=&CustomRotator}.Rotated Values
   Size=LinearInterpolate(A=Constant2Vector() {g=0.1, r=0.3}, B=Constant2Vector() {g=0.1, r=0.8}, Alpha=LinearInterpolate(Alpha=@n27) {constA=-0.75})
   CornerRoundness=Constant() {r=0.05}
  ) {materialFunction=&MF_UI_Box}
  NamedRerouteDeclaration(None=SmoothStep(Value=@n28.SDF) {constMax=-0.006, constMin=0.004}) {name=Bottom_Pips, nodeColor=<rgba 0 0 0 1>}
  NamedRerouteDeclaration(None=Add(A=NamedRerouteUsage(), B=NamedRerouteUsage())) {name=Pips, nodeColor=<rgba 0 0 0 1>}
  NamedRerouteDeclaration(None=SmoothStep(Value=@n28.SDF) {constMax=0.02, constMin=0.1}) {name=Bottom_Pips_DropShadow, nodeColor=<rgba 0 0 0 1>}
  NamedRerouteDeclaration(None=@n18.DotWithGlow) {name=CenterDot, nodeColor=<rgba 0 0 0 1>}
 ]
}
