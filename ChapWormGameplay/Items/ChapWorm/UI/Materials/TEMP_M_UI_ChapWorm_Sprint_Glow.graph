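# Temporary translucent UI material, named as the glow for the ChapWorm sprint interface.
# An MF_UI_Box function draws a procedural box fill from texture coordinates, with constants setting its size, corner roundness and glow range.
# A Blend_Overlay function blends that fill over a Color vector parameter, orange by default, so the shape needs no texture.
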Material /ChapWormGameplay/Items/ChapWorm/UI/Materials/TEMP_M_UI_ChapWorm_Sprint_Glow
graph {
 properties {blendMode=BLEND_Translucent, materialDomain=MD_UI, pixelDepthOffsetMode=PDOM_Legacy, usedShadingModels=Default Lit}
 nodes [MaterialFunctionCall(
   Base=VectorParameter() {parameterName=Color, defaultValue=<rgba 1 0.2184 0 1>}.RGB
   Blend=MaterialFunctionCall(UVs=TextureCoordinate(), Size=Constant() {r=0.1}, CornerRoundness=Constant(), GlowMin=Constant(), GlowMax=Constant() {r=0.9}) {materialFunction=&MF_UI_Box}.Fill
  ) {materialFunction=&Blend_Overlay}]
}
