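# Additive UI material for the ChapWorm super-charged charge bar, with animated plasma energy in purple tones.
# A charge-fill texture supplies the bar shape, while FXProgress interpolates between ProgressStart and ProgressEnd to place a soft-edged gradient and top glow bands.
# Panned plasma-coil samples form left and right noise, with a horizontal Mirror mask for the split, alongside a radial swirl.
# Named reroutes organize these layers.
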Material /ChapWormGameplay/Items/ChapWorm/UI/Materials/M_UI_ChapWorm_SuperCharged
graph {
 properties {bAutomaticallySetUsageInEditor=false, blendMode=BLEND_Additive, materialDomain=MD_UI, usedShadingModels=Default Lit}
 nodes [
  LinearInterpolate(A=Constant3Vector() {constant=<rgba 0.219 0 0.9401 1>}, B=Constant3Vector() {constant=<rgba 1.4334 0.15 3 1>}, Alpha=NamedRerouteUsage())
  MaterialFunctionCall(Speed=Constant() {r=0.2}) {materialFunction=&DebugTimeSine}
  n1 = NamedRerouteUsage()
  n2 = Multiply(A=@n1, B=@n1) {constB=2.9687}
  n3 = NamedRerouteUsage()
  n4 = Multiply(A=Multiply(A=@n3, B=@n3)) {constB=0.032}
  Multiply(A=Multiply(A=Multiply(A=@n2, B=@n4), B=NamedRerouteUsage())) {constB=0.8}
  n5 = MaterialFunctionCall(UVs=Add(A=TextureCoordinate(), B=Constant2Vector() {g=0.015, r=0.02}) {constB=0}, Texture Scale=Constant() {r=0.68}) {materialFunction=&ScaleUVsByCenter}
  NamedRerouteDeclaration(None=SmoothStep(Value=TextureSample(UVs=@n5.UVs) {samplerType=SAMPLERTYPE_LinearGrayscale, texture=/ChapWormGameplay/Items/ChapWorm/UI/Textures/T_UI_ChapWorm_ChargeFill}.R)) {name=ProgressBar, nodeColor=<rgba 0 0.2734 1 1>}
  n6 = TextureCoordinate()
  n7 = Constant() {r=0.5}
  n8 = ScalarParameter() {
   defaultValue=0.2284
   desc="When progress value changes, it defines where change starts.\r\n0 - 1"
   group=Progress
   parameterName=ProgressStart
   sliderMax=1
   sortPriority=1
  }
  n9 = ScalarParameter() {
   defaultValue=0.81
   desc="When progress value changes, it defines where change ends.\r\n0 - 1"
   group=Progress
   parameterName=ProgressEnd
   sliderMax=1
   sortPriority=2
  }
  n10 = ScalarParameter() {
   defaultValue=1
   desc="Property for animating progress value change (from ProgressStart to ProgressEnd)\r\n0 - 1"
   group=Progress
   parameterName=FXProgress
   sliderMax=1
   sortPriority=3
  }
  n11 = Add(
   A=Multiply(A=MaterialFunctionCall(UVs=@n6, "Rotation Angle (0-1)"=@n7) {materialFunction=&CustomRotator}.Rotated Values) {constB=-1}
   B=NamedRerouteDeclaration(None=LinearInterpolate(A=@n8, B=@n9, Alpha=@n10)) {name=Progress, nodeColor=<rgba 0 0 0 1>}
  ) {constB=0.5}
  n12 = SmoothStep(
   Max=ScalarParameter() {defaultValue=0.1436, group=Progress, parameterName=ProgressEdgeSharpness, sliderMax=1, sortPriority=2}
   Value=ComponentMask(None=@n11) {g=true}
  ) {constMax=0.01}
  n13 = OneMinus(None=ComponentMask(None=@n11) {g=true})
  n14 = Multiply(A=@n12, B=Subtract(A=@n13) {constB=0.85})
  NamedRerouteDeclaration(None=Saturate(None=Multiply(A=@n14) {constB=5})) {name=ProgressTopGlow, nodeColor=<rgba 0 0 0 1>}
  NamedRerouteDeclaration(None=Saturate(None=@n12)) {name=ProgressGradient, nodeColor=<rgba 0 0 0 1>}
  n15 = MaterialFunctionCall!(X Scale=Constant() {r=2}, Time, Y Scale=Constant() {r=1}, Flip UVs, X Speed=Constant() {r=-0.268}, Y Speed=Constant() {r=0.1}, UV) {materialFunction=&/MF_RadialPanner_ChapWorm}
  NamedRerouteDeclaration(None=Multiply(A=TextureSample(UVs=@n15.Result) {samplerType=SAMPLERTYPE_LinearGrayscale, texture=/Game/Effects/Fort_Effects/Emotes/UltraLight/Texture/T_PlasmaCoils}.R) {constB=8}) {name=RadialSwirl, nodeColor=<rgba 0 0 0 1>}
  n16 = Multiply(A=@n12, B=Subtract(A=@n13) {constB=0.97})
  NamedRerouteDeclaration(None=Saturate(None=Multiply(A=@n16) {constB=40})) {name=ProgressTopGlowShine, nodeColor=<rgba 0 0 0 1>}
  n17 = TextureCoordinate() {uTiling=1.5, vTiling=1.5}
  n18 = Constant() {r=-0.7}
  n19 = Constant()
  n20 = TextureSample(UVs=Panner(Coordinate=@n17, Speed=AppendVector(A=@n18, B=@n19)) {speedX=-0.075, speedY=0.125}) {samplerType=SAMPLERTYPE_LinearGrayscale, texture=&T_PlasmaCoils}
  n21 = TextureCoordinate()
  n22 = Constant() {r=-0.3}
  n23 = Constant()
  n24 = TextureSample(UVs=Panner(Coordinate=@n21, Speed=AppendVector(A=@n22, B=@n23)) {speedX=-0.075, speedY=0.125}) {samplerType=SAMPLERTYPE_LinearGrayscale, texture=&T_PlasmaCoils}
  NamedRerouteDeclaration(None=Multiply(A=DotProduct(A=@n20.R, B=@n24.R)) {constB=4}) {name=RightNoise, nodeColor=<rgba 0 0 0 1>}
  n25 = AppendVector(A=Constant() {r=0.2}, B=Constant())
  n26 = AppendVector(A=Constant() {r=0.7}, B=Constant())
  n27 = DotProduct(
   A=TextureSample(UVs=Panner(Coordinate=@n17, Speed=@n25) {speedX=-0.075, speedY=0.125}) {samplerType=SAMPLERTYPE_LinearGrayscale, texture=&T_PlasmaCoils}.R
   B=TextureSample(UVs=Panner(Coordinate=@n21, Speed=@n26) {speedX=-0.075, speedY=0.125}) {samplerType=SAMPLERTYPE_LinearGrayscale, texture=&T_PlasmaCoils}.R
  )
  NamedRerouteDeclaration(None=Saturate(None=Multiply(A=@n27) {constB=4})) {name=LeftNoise, nodeColor=<rgba 0 0 0 1>}
  n28 = ComponentMask(None=Add(A=TextureCoordinate()) {constB=-0.5}) {r=true}
  NamedRerouteDeclaration(None=Saturate(None=Multiply(A=@n28) {constB=10})) {name=Mirror, nodeColor=<rgba 0 0 0 1>}
  n29 = NamedRerouteUsage()
  n30 = NamedRerouteUsage()
  n31 = NamedRerouteUsage()
  NamedRerouteDeclaration(None=Multiply(A=LinearInterpolate(A=@n29, B=@n30, Alpha=@n31), B=NamedRerouteUsage())) {name=NoisyNeighbors, nodeColor=<rgba 1 0.4453 0 1>}
 ]
}
