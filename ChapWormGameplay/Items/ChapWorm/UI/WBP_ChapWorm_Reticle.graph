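# Reticle extension widget for the ChapWorm weapon, showing kinetic charge, in-range targets, boosted charging, pounce targeting and Overpower state.
# Construct binds weapon ammo and targeting component delegates, plus async gameplay tag and event listeners on the view target pawn that trigger state animations.
# Charge progress follows magazine ammo; during Overpower, Tick replaces it with the active effect's remaining duration fraction.
# Values feed dynamic material parameters, and Destruct unbinds delegates and ends async listeners.
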WidgetBlueprint /ChapWormGameplay/Items/ChapWorm/UI/WBP_ChapWorm_Reticle
animations [
 Intro {
  binding_count=1
  length_seconds=0.55
  object=&self:Intro
  sequence {bindings [Reticle {
     id=#1
     possessed_class=Overlay
     tracks [{
       type=MovieSceneFloatTrack
       display_name=Render Opacity
       sections [{type=MovieSceneFloatSection, end_frame=11, start_frame=0, channels [None {type=MovieSceneScriptingFloatChannel, keys [[0,0],[11,1]]}]}]
      }]
    }]}
 }
 InRange {
  binding_count=1
  length_seconds=0.25
  object=&self:InRange
  sequence {bindings [Image_Reticle {
     id=#2
     possessed_class=Image
     tracks [{
       type=MovieSceneWidgetMaterialTrack
       sections [{
         type=MovieSceneParameterSection
         end_frame=5
         start_frame=0
         channels [
          ChangetoReach {type=MovieSceneScriptingFloatChannel, keys [[0,0],[2.5,0.4974],[5.0173,1]]}
          ReticleScale {type=MovieSceneScriptingFloatChannel, const [0.9,1]}
          CenterScale {type=MovieSceneScriptingFloatChannel, keys [[0,0.3],[5,0.4]]}
          InRange_Switch {type=MovieSceneScriptingFloatChannel, keys [[0,0],[5,1]]}
         ]
        }]
      }]
    }]}
 }
 BoostedCharging {
  binding_count=1
  length_seconds=0.3334
  object=&self:BoostedCharging
  sequence {
   bindings [Image_ChargeUp_Base {
     id=#3
     possessed_class=Image
     tracks [{
       type=MovieSceneWidgetMaterialTrack
       sections [{
         type=MovieSceneParameterSection
         end_frame=10
         start_frame=0
         channels [BoostedChargeFX_Switch {type=MovieSceneScriptingFloatChannel, keys [[0,0],[10,1]]}]
        }]
      }]
    }]
   display_rate {numerator=30}
  }
 }
 ReadyforUse {
  binding_count=5
  length_seconds=1
  object=&self:ReadyforUse
  sequence {bindings [
    Image_ChargeUp_Lightning {
     id=#4
     possessed_class=Image
     tracks [{
       type=MovieSceneFloatTrack
       display_name=Render Opacity
       sections [{type=MovieSceneFloatSection, end_frame=6, start_frame=3, channels [None {type=MovieSceneScriptingFloatChannel, keys [[3,0],[6,1]]}]}]
      }]
    }
    Image_ChargeUp_Base {
     id=#5
     possessed_class=Image
     tracks [{
       type=MovieSceneWidgetMaterialTrack
       sections [{
         type=MovieSceneParameterSection
         end_frame=20
         start_frame=0
         channels [
          Progress {type=MovieSceneScriptingFloatChannel, const [1.1,1]}
          Charged_IdleFX {type=MovieSceneScriptingFloatChannel, keys [[0,0],[5,1],[20,1]]}
         ]
        }]
      }]
    }
    Image_Reticle {
     id=#6
     possessed_class=Image
     tracks [{
       type=MovieSceneWidgetMaterialTrack
       sections [{
         type=MovieSceneParameterSection
         end_frame=19
         start_frame=0
         channels [
          Pip_LockInSwitch {type=MovieSceneScriptingFloatChannel, keys [[0,0],[6,1]]}
          Pip_StretchSwitch {type=MovieSceneScriptingFloatChannel, keys [[0,0.3],[6,1]]}
          BottomPip_ReadySwitch {type=MovieSceneScriptingFloatChannel, keys [[0,0],[6,1]]}
          PipReadyFlash {type=MovieSceneScriptingFloatChannel, keys [[5,0],[7,1]]}
          UseMeflash {type=MovieSceneScriptingFloatChannel, keys [[0,0],[16,0],[19,1]]}
         ]
        }]
      }]
    }
    Image_ChargeUp_Additive {
     id=#7
     possessed_class=Image
     tracks [{
       type=MovieSceneFloatTrack
       display_name=Render Opacity
       sections [{type=MovieSceneFloatSection, end_frame=16, start_frame=0, channels [None {type=MovieSceneScriptingFloatChannel, const [0,3]}]}]
      }]
    }
    Image_Glow {
     id=#8
     possessed_class=Image
     tracks [
      {
       type=MovieSceneFloatTrack
       display_name=Render Opacity
       sections [{
         type=MovieSceneFloatSection
         end_frame=10
         start_frame=0
         channels [None {type=MovieSceneScriptingFloatChannel, keys [[0,0],[3,1.1463],[10,0]]}]
        }]
      }
      {
       type=MovieScene2DTransformTrack
       sections [{
         type=MovieScene2DTransformSection
         end_frame=10
         start_frame=0
         channels [
          Translation.X {type=MovieSceneScriptingFloatChannel}
          Translation.Y {type=MovieSceneScriptingFloatChannel}
          Angle {type=MovieSceneScriptingFloatChannel}
          Scale.X {type=MovieSceneScriptingFloatChannel, keys [[0,1.0976],[3,0.87]]}
          Scale.Y {type=MovieSceneScriptingFloatChannel, keys [[0,0],[10,1.5943]]}
          Shear.X {type=MovieSceneScriptingFloatChannel}
          Shear.Y {type=MovieSceneScriptingFloatChannel}
         ]
        }]
      }
     ]
    }
   ]}
 }
 Activated {
  binding_count=4
  length_seconds=0.4214
  object=&self:Activated
  sequence {bindings [
    Image_Glow {
     id=#9
     possessed_class=Image
     tracks [
      {
       type=MovieSceneFloatTrack
       display_name=Render Opacity
       sections [{
         type=MovieSceneFloatSection
         end_frame=8
         start_frame=0
         channels [None {type=MovieSceneScriptingFloatChannel, keys [[0,0],[3.483,0.54],[8.4287,0]]}]
        }]
      }
      {
       type=MovieScene2DTransformTrack
       sections [{
         type=MovieScene2DTransformSection
         end_frame=8
         start_frame=0
         channels [
          Translation.X {type=MovieSceneScriptingFloatChannel, const [76,1]}
          Translation.Y {type=MovieSceneScriptingFloatChannel, const [45,1]}
          Angle {type=MovieSceneScriptingFloatChannel, const [35,1]}
          Scale.X {type=MovieSceneScriptingFloatChannel, const [1,3]}
          Scale.Y {type=MovieSceneScriptingFloatChannel, keys [[3.483,1.4887],[8.4287,0]]}
          Shear.X {type=MovieSceneScriptingFloatChannel, const [0,1]}
          Shear.Y {type=MovieSceneScriptingFloatChannel, const [0,1]}
         ]
        }]
      }
     ]
    }
    Image_ChargeUp_Lightning {
     id=#10
     possessed_class=Image
     tracks [{
       type=MovieSceneFloatTrack
       display_name=Render Opacity
       sections [{type=MovieSceneFloatSection, end_frame=4, start_frame=0, channels [None {type=MovieSceneScriptingFloatChannel, keys [[0,1],[4,0]]}]}]
      }]
    }
    Image_ChargeUp_Base {
     id=#11
     possessed_class=Image
     tracks [{type=MovieSceneWidgetMaterialTrack, sections [{type=MovieSceneParameterSection, end_frame=8, start_frame=0}]}]
    }
    Image_Reticle {
     id=#12
     possessed_class=Image
     tracks [{
       type=MovieSceneWidgetMaterialTrack
       sections [{
         type=MovieSceneParameterSection
         end_frame=8
         start_frame=0
         channels [
          Pip_LockInSwitch {type=MovieSceneScriptingFloatChannel, keys [[0,1],[5,1],[8.425,0]]}
          Pip_StretchSwitch {type=MovieSceneScriptingFloatChannel, const [1,3]}
          BottomPip_ReadySwitch {type=MovieSceneScriptingFloatChannel, keys [[0,1],[5,1],[8.4287,0]]}
          PipReadyFlash {type=MovieSceneScriptingFloatChannel, keys [[0,1],[5,0]]}
         ]
        }]
      }]
    }
   ]}
 }
 Empty {
  binding_count=1
  length_seconds=0.2
  object=&self:Empty
  sequence {bindings [Image_Reticle {
     id=#13
     possessed_class=Image
     tracks [{
       type=MovieSceneWidgetMaterialTrack
       sections [{
         type=MovieSceneParameterSection
         end_frame=4
         start_frame=0
         channels [Pip_StretchSwitch {type=MovieSceneScriptingFloatChannel, keys [[0,1],[4,0]]}]
        }]
      }]
    }]}
 }
 SuperChargeStateDepleting {
  binding_count=2
  length_seconds=0.6
  object=&self:SuperChargeStateDepleting
  sequence {bindings [
    Image_ChargeUp_Base {
     id=#14
     possessed_class=Image
     tracks [{
       type=MovieSceneWidgetMaterialTrack
       sections [{
         type=MovieSceneParameterSection
         end_frame=12
         start_frame=0
         channels [SCDepleteGlow_Switch {type=MovieSceneScriptingFloatChannel, keys [[0,0],[12,1]]}]
        }]
      }]
    }
    Image_Glow {
     id=#15
     possessed_class=Image
     tracks [{
       type=MovieSceneFloatTrack
       display_name=Render Opacity
       sections [{
         type=MovieSceneFloatSection
         end_frame=12
         start_frame=0
         channels [None {type=MovieSceneScriptingFloatChannel, keys [[0,0],[6,0.21],[12,0]]}]
        }]
      }]
    }
   ]}
 }
 Pounce {
  binding_count=1
  length_seconds=0.2
  object=&self:Pounce
  sequence {
   bindings [Image_Reticle {
     id=#16
     possessed_class=Image
     tracks [{
       type=MovieScene2DTransformTrack
       sections [{
         type=MovieScene2DTransformSection
         end_frame=6
         start_frame=0
         channels [
          Translation.X {type=MovieSceneScriptingFloatChannel, const [0,1]}
          Translation.Y {type=MovieSceneScriptingFloatChannel, const [0,1]}
          Angle {type=MovieSceneScriptingFloatChannel, const [0,1]}
          Scale.X {type=MovieSceneScriptingFloatChannel, keys [[0,1],[6,1.1]]}
          Scale.Y {type=MovieSceneScriptingFloatChannel, keys [[0,1],[6,1.1]]}
          Shear.X {type=MovieSceneScriptingFloatChannel, const [0,1]}
          Shear.Y {type=MovieSceneScriptingFloatChannel, const [0,1]}
         ]
        }]
      }]
    }]
   display_rate {numerator=30}
  }
 }
]
graphs [
 EventGraph {nodes [
   Event Destruct()
   > n1 = Sequence()
   > n2 = Get!(execute:Exec) -> (then:Exec, else:Exec, EquippedWeapon:Fort Weapon Object Reference)
   > n3 = Unbind Event from On Local Ammo Changed(self:Fort Weapon Object Reference=@n2, Delegate=Create Event()) {is=title K2Node_RemoveDelegate}
   n4 = Get!(execute=[@n3.then, @n2.else]) -> (then, else, TargetingComponent:B Chap Worm Targeting Component Object Reference)
   > n5 = Unbind Event from On Valid Target in Range Changed(self:B Chap Worm Targeting Component Object Reference=@n4, Delegate=Create Event()) {is=title K2Node_RemoveDelegate}
   EndAction(
    execute=[@n5.then, @n4.else]
    self:Ability Async Object Reference=Get AsyncActions() -> (AsyncActions:Array of Ability Async Object References)
   )
   Parent: Destruct(execute=@n1.then_1)
   Event Construct()
   > Parent: Construct()
   n6 = GetOwningPlayer()
   > Is Valid(InputObject=@n6)
   > n7 = Cast To PlayerPawn_Athena(Object=GetViewTarget!(self:Controller Object Reference=@n6) -> (ReturnValue:Actor Object Reference))
   > Set ViewTargetPlayerPawn(ViewTargetPlayerPawn:Player Pawn Athena Object Reference=@n7) -> (Output_Get:Player Pawn Athena Object Reference)
   > n8 = Set EquippedWeapon(EquippedWeapon:Fort Weapon Object Reference=GetEquippedWeapon!(self:Fort Weapon Reticle Extension Widget Base Object Reference) -> (ReturnValue:Fort Weapon Object Reference)) -> (Output_Get:Fort Weapon Object Reference)
   > Is Valid(InputObject=@n8)
   > Bind Event to On Local Ammo Changed(self:Fort Weapon Object Reference=@n8, Delegate=Create Event()) {is=title K2Node_AddDelegate}
   > n9 = Sequence()
   > n10 = Set TargetingComponent(TargetingComponent:B Chap Worm Targeting Component Object Reference=GetComponentByClass!(
     self:Actor Object Reference=Get EquippedWeapon() -> (EquippedWeapon:Fort Weapon Object Reference)
     ComponentClass:Actor Component Class Reference=/ChapWormGameplay/Items/ChapWorm/B_ChapWormTargetingComponent.B_ChapWormTargetingComponent_C
    ) -> (ReturnValue:B Chap Worm Targeting Component Object Reference)) -> (Output_Get:B Chap Worm Targeting Component Object Reference)
   > Is Valid(InputObject=@n10)
   > Bind Event to On Valid Target in Range Changed(self:B Chap Worm Targeting Component Object Reference=@n10, Delegate=Create Event()) {is=title K2Node_AddDelegate}
   n11 = Get ViewTargetPlayerPawn() -> (ViewTargetPlayerPawn:Player Pawn Athena Object Reference)
   n12 = Get BoostedChargingTag() -> (BoostedChargingTag:Gameplay Tag Structure)
   > n13 = WaitGameplayTagAddToActor!(execute:Exec, TargetActor:Actor Object Reference=@n11, Tag:Gameplay Tag Structure=@n12, OnlyTriggerOnce:bool=false) -> (then:Exec, AsyncTaskProxy:Ability Async Wait Gameplay Tag Added Object Reference, Added:Exec) {is=title K2Node_AsyncAction}
   > Add(TargetArray:Array of Ability Async Object References=Get AsyncActions(), NewItem:"Ability Async Object Reference (by ref)"=@n13) {is=title K2Node_CallArrayFunction}
   > n14 = WaitGameplayTagRemoveFromActor!(execute:Exec, TargetActor:Actor Object Reference=@n11, Tag:Gameplay Tag Structure=@n12, OnlyTriggerOnce:bool=false) -> (then:Exec, AsyncTaskProxy:Ability Async Wait Gameplay Tag Removed Object Reference, Removed:Exec) {is=title K2Node_AsyncAction}
   > Add(TargetArray=Get AsyncActions(), NewItem=@n14)
   n15 = Get OverpowerActiveTag() -> (OverpowerActiveTag:Gameplay Tag Structure)
   > n16 = WaitGameplayTagAddToActor(TargetActor=@n11, Tag=@n15, OnlyTriggerOnce=false)
   > Add(TargetArray=Get AsyncActions(), NewItem=@n16)
   > n17 = WaitGameplayTagRemoveFromActor(TargetActor=@n11, Tag=@n15, OnlyTriggerOnce=false)
   > Add(TargetArray=Get AsyncActions(), NewItem=@n17)
   n18 = Get PounceTargetingTag() -> (PounceTargetingTag:Gameplay Tag Structure)
   > n19 = WaitGameplayTagAddToActor(TargetActor=@n11, Tag=@n18, OnlyTriggerOnce=false)
   > Add(TargetArray=Get AsyncActions(), NewItem=@n19)
   > n20 = WaitGameplayTagRemoveFromActor(TargetActor=@n11, Tag=@n18, OnlyTriggerOnce=false)
   > Add(TargetArray=Get AsyncActions(), NewItem=@n20)
   > n21 = WaitGameplayTagAddToActor(TargetActor=@n11, Tag=Get KineticEnergyMaxedTag() -> (KineticEnergyMaxedTag:Gameplay Tag Structure), OnlyTriggerOnce=false)
   > Add(TargetArray=Get AsyncActions(), NewItem=@n21)
   > n22 = WaitGameplayTagRemoveFromActor(TargetActor=@n11, Tag=Get KineticEnergyMaxedTag(), OnlyTriggerOnce=false)
   > Add(TargetArray=Get AsyncActions(), NewItem=@n22)
   > n23 = WaitGameplayEventToActor!(
    execute:Exec
    TargetActor:Actor Object Reference=@n11
    EventTag:Gameplay Tag Structure=(TagName="ChapWorm.Event.Overpower.Refresh")
    OnlyTriggerOnce:bool=false
    OnlyMatchExact:bool=true
   ) -> (then:Exec, AsyncTaskProxy:Ability Async Wait Gameplay Event Object Reference, EventReceived:Exec, Payload:Gameplay Event Data Structure) {is=title K2Node_AsyncAction}
   > Add(TargetArray=Get AsyncActions(), NewItem=@n23)
   HandleOverpower!(execute=@n23.EventReceived, self, IsOverpowerActive:bool=true) -> (then)
   Branch(execute=@n22.Removed, Condition=Get IsReadyForUse() -> (IsReadyForUse:bool))
   n24 = Get ReadyforUse() -> (ReadyforUse:Widget Animation Object Reference)
   > PlayAnimation(InAnimation=@n24, StartAtTime=GetEndTime(self=@n24), PlayMode=Reverse)
   > Set IsReadyForUse(IsReadyForUse:bool=false) -> (Output_Get:bool)
   n25 = Branch(execute=@n21.Added, Condition=Get IsReadyForUse())
   PlayAnimation(execute=@n25.else, InAnimation=Get ReadyforUse())
   > Set IsReadyForUse(IsReadyForUse=true)
   HandlePounceTargeting!(execute=@n20.Removed, self, bPounceTargeting:bool=false) -> (then)
   HandlePounceTargeting(execute=@n19.Added, bPounceTargeting=true)
   HandleOverpower(execute=@n17.Removed, IsOverpowerActive=false)
   HandleOverpower(execute=@n16.Added, IsOverpowerActive=true)
   HandleBoostedCharging!(execute=@n14.Removed, self, bIsBoostedCharging:bool=false) -> (then)
   HandleBoostedCharging(execute=@n13.Added, bIsBoostedCharging=true)
   HandleAmmoCountChanged!(
    execute=@n9.then_1
    self
    LocalCount:int=0
    LocalRemaining:int=GetLocalMagazineAmmoCount!(self:Fort Weapon Object Reference=@n8) -> (ReturnValue:int=0)
   ) -> (then)
   n26 = Event Tick()
   > Parent: Tick!(execute, MyGeometry:Geometry Structure=@n26.MyGeometry, InDeltaTime:float=@n26.InDeltaTime) -> (then) {is=title K2Node_CallParentFunction}
   > Branch(Condition=Get bOverpowerActive() -> (bOverpowerActive:bool))
   > UpdateOverpowerDurationVisual()
   n27 = Get Image_ChargeUp_Additive() -> (Image_ChargeUp_Additive:Image Object Reference)
   > n28 = GetDynamicMaterial(self=@n27)
   n29 = Get EffectDuration() -> (EffectDuration:double)
   > SetScalarParameterValue(self=@n28, ParameterName=FXProgress, Value=@n29)
   > SetRenderOpacity(self=@n27, InOpacity=@n29)
   n30 = Event PreConstruct()
   > Parent: PreConstruct(IsDesignTime=@n30.IsDesignTime)
   SetChargeProgress!(execute, self, Value:float=0) -> (then)
  ]}
 HandleAmmoCountChanged {nodes [
   n1 = HandleAmmoCountChanged!() -> (then, LocalCount:int, LocalRemaining:int) {is=title K2Node_FunctionEntry}
   > Branch(Condition=NOT Boolean(A=Get bOverpowerActive()))
   > UpdateChargeProgress!(execute, self, Alpha:float=SafeDivide(A="To Float (Integer)"(InInt=@n1.LocalRemaining), B=100)) -> (then)
  ]}
 HandleBoostedCharging {nodes [
   n1 = HandleBoostedCharging!() -> (then, bIsBoostedCharging:bool=~) {is=title K2Node_FunctionEntry}
   > n2 = Branch(Condition=@n1)
   n3 = Get BoostedCharging() -> (BoostedCharging:Widget Animation Object Reference)
   > PlayAnimation(InAnimation=@n3)
   PlayAnimation(execute=@n2.else, InAnimation=@n3, PlayMode=Reverse)
  ]}
 HandleOverpower {nodes [
   n1 = HandleOverpower!() -> (then, IsOverpowerActive:bool=~) {is=title K2Node_FunctionEntry}
   > n2 = Set bOverpowerActive(bOverpowerActive:bool=@n1) -> (Output_Get:bool)
   > n3 = Branch(Condition=@n2)
   > PlayAnimation(InAnimation=Get Activated() -> (Activated:Widget Animation Object Reference))
   > PlayAnimation(InAnimation=Get SuperChargeStateDepleting() -> (SuperChargeStateDepleting:Widget Animation Object Reference))
   n4 = Get SuperChargeStateDepleting()
   PlayAnimation(execute=@n3.else, InAnimation=@n4, StartAtTime=GetEndTime(self=@n4), PlayMode=Reverse)
  ]}
 HandlePounceTargeting {nodes [
   n1 = HandlePounceTargeting!() -> (then, bPounceTargeting:bool=~) {is=title K2Node_FunctionEntry}
   > n2 = Branch(Condition=@n1)
   n3 = Get Pounce() -> (Pounce:Widget Animation Object Reference)
   > PlayAnimation(InAnimation=@n3)
   PlayAnimation(execute=@n2.else, InAnimation=@n3, PlayMode=Reverse)
  ]}
 HandleValidTargetInRangeChanged {nodes [
   n1 = HandleValidTargetInRangeChanged!() -> (then, bValid:bool=~) {is=title K2Node_FunctionEntry}
   > n2 = GetDynamicMaterial(self=Get Image_Reticle() -> (Image_Reticle:Image Object Reference))
   > Is Valid(InputObject=@n2)
   > SetScalarParameterValue(self=@n2, ParameterName=InRange_Switch, Value="To Float (Boolean)"(InBool=@n1))
  ]}
 SetChargeProgress {nodes [
   n1 = SetChargeProgress!() -> (then, Value:float=~) {is=title K2Node_FunctionEntry}
   > n2 = GetDynamicMaterial(self=Get Image_ChargeUp_Base() -> (Image_ChargeUp_Base:Image Object Reference))
   > Is Valid(InputObject=@n2)
   > SetScalarParameterValue(self=@n2, ParameterName=Progress, Value=@n1)
   > n3 = GetDynamicMaterial(self=Get Image_Reticle())
   > Is Valid(InputObject=@n3)
   > SetScalarParameterValue(self=@n3, ParameterName=Pip_StretchSwitch, Value=@n1)
   Reroute Node(InputPin:Wildcard) -> (OutputPin:Wildcard)
  ]}
 UpdateChargeProgress {nodes [
   n1 = UpdateChargeProgress!() -> (then, Alpha:float=~) {is=title K2Node_FunctionEntry}
   > n2 = Set EffectDuration(EffectDuration:double=@n1) -> (Output_Get:double)
   > SetChargeProgress(Value=@n2) {is=title K2Node_CallFunction}
   > n3 = GetDynamicMaterial(self=Get Image_Reticle())
   > Is Valid(InputObject=@n3)
   > n4 = Branch(Condition="Nearly Equal (Float)"(A=@n1, B=1, ErrorTolerance=0))
   > SetScalarParameterValue(self=@n3, ParameterName=BottomPip_ReadySwitch, Value=1)
   n5 = Branch(execute=@n4.else, Condition=IsAnimationPlaying(InAnimation=Get Activated()))
   SetScalarParameterValue(execute=@n5.else, self=@n3, ParameterName=BottomPip_ReadySwitch, Value=0)
  ]}
 UpdateOverpowerDurationVisual {nodes [
   UpdateOverpowerDurationVisual() {is=title K2Node_FunctionEntry}
   > n1 = Get!(execute) -> (then, else, ViewTargetPlayerPawn:Player Pawn Athena Object Reference)
   n2 = GetFortAbilitySystemComponent!(Actor:Actor Object Reference=@n1) -> (ReturnValue:Fort Ability System Component Object Reference)
   > Is Valid(InputObject=@n2)
   > n3 = GetActiveEffectsWithAllTags!(
    execute
    self:Ability System Component Object Reference=@n2
    Tags:Gameplay Tag Container Structure=MakeGameplayTagContainerFromTag!(SingleTag:Gameplay Tag Structure=Get OverpowerActiveTag()) -> (ReturnValue:Gameplay Tag Container Structure)
   ) -> (then, ReturnValue:Array of Active Gameplay Effect Handle Structures)
   > Branch(Condition=Is Valid Index!(TargetArray:Array of Active Gameplay Effect Handle Structures=@n3, IndexToTest:int) -> (ReturnValue:bool))
   n4 = "Get (a ref)"(Array:Array of Active Gameplay Effect Handle Structures=@n3, Dimension 1=0) -> (Output:"Active Gameplay Effect Handle Structure (by ref)")
   > n5 = GetActiveGameplayEffectRemainingDuration!(execute, ActiveHandle:Active Gameplay Effect Handle Structure=@n4) -> (then, ReturnValue:float)
   > n6 = GetActiveGameplayEffectTotalDuration!(execute, ActiveHandle:Active Gameplay Effect Handle Structure=@n4) -> (then, ReturnValue:float)
   > UpdateChargeProgress(Alpha=SafeDivide(A=@n5, B=@n6)) {is=title K2Node_CallFunction}
  ]}
]
variables [
 AsyncActions {type="array<AbilityAsync>", default []}
 BoostedChargingTag {type=GameplayTag, default {tagName=ChapWorm.KineticEnergy.Status.BoostedCharging}}
 bOverpowerActive {type=boolean}
 EffectDuration {type=number}
 EquippedWeapon {type=FortWeapon, default=null}
 IsReadyForUse {type=boolean}
 KineticEnergyMaxedTag {type=GameplayTag, default {tagName=ChapWorm.KineticEnergy.Maxed}}
 OverpowerActiveTag {type=GameplayTag, default {tagName=ChapWorm.ApplyOverpowerStatus}}
 PounceTargetingTag {type=GameplayTag, default {tagName=ChapWorm.Pounce.Targeting}}
 TargetingComponent {type=B_ChapWormTargetingComponent_C, default=null}
 ViewTargetPlayerPawn {type=PlayerPawn_Athena_C}
]
widgets {
 info {
  inheritedWidgetCount=1
  parentClass=/Script/FortniteUI.FortWeaponReticleExtensionWidgetBase
  rootWidgetClass=&Overlay
  widgetCount=7
 }
 widgets [
  Reticle {widgetClassPath=&Overlay, widget=Reticle}
  Image_Reticle {
   widgetClassPath=&Image
   bIsVariable=true
   parent=Reticle
   widget=Image_Reticle
   properties {brush {
     bIsDynamicallyLoaded=false
     drawAs=Image
     imageType=FullColor
     mirroring=NoMirror
     resourceName=None
     resourceObject=/ChapWormGameplay/Items/ChapWorm/UI/Materials/MI_UI_ChapWorm_Reticle
     tiling=NoTile
     imageSize=<100 100>
     margin=<ltrb 0 0 0 0>
     outlineSettings {
      bUseBrushTransparency=false
      roundingType=HalfHeightRadius
      width=0
      color {colorUseRule=UseColor_Specified, specifiedColor=<rgba 0 0 0 0>}
      cornerRadii=<0 0 0 0>
     }
     tintColor {colorUseRule=UseColor_Specified, specifiedColor=<rgba 1 1 1 1>}
     uVRegion {bIsValid=false, max=<0 0>, min=<0 0>}
    }}
   slot {class=OverlaySlot}
  }
  Image_ChargeUp_Base {
   widgetClassPath=&Image
   bIsVariable=true
   parent=Reticle
   widget=Image_ChargeUp_Base
   properties {
    brush {
     bIsDynamicallyLoaded=false
     drawAs=Image
     imageType=FullColor
     mirroring=NoMirror
     resourceName=None
     resourceObject=/ChapWormGameplay/Items/ChapWorm/UI/Materials/MI_UI_ChapWorm_ChargeUp
     tiling=NoTile
     imageSize=<32 32>
     margin=<ltrb 0 0 0 0>
     outlineSettings {
      bUseBrushTransparency=false
      roundingType=HalfHeightRadius
      width=0
      color {colorUseRule=UseColor_Specified, specifiedColor=<rgba 0 0 0 0>}
      cornerRadii=<0 0 0 0>
     }
     tintColor {colorUseRule=UseColor_Specified, specifiedColor=<rgba 1 1 1 1>}
     uVRegion {bIsValid=false, max=<0 0>, min=<0 0>}
    }
    renderTransform {angle=0, scale=<1 1>, shear=<0 0>, translation=<73 47>}
   }
   slot {class=OverlaySlot, properties {horizontalAlignment=HAlign_Fill, verticalAlignment=VAlign_Fill}}
  }
  Image_ChargeUp_Additive {
   widgetClassPath=&Image
   bIsVariable=true
   parent=Reticle
   widget=Image_ChargeUp_Additive
   properties {
    renderOpacity=0
    brush {
     bIsDynamicallyLoaded=false
     drawAs=Image
     imageType=FullColor
     mirroring=NoMirror
     resourceName=None
     resourceObject=/ChapWormGameplay/Items/ChapWorm/UI/Materials/M_UI_ChapWorm_SuperCharged
     tiling=NoTile
     imageSize=<32 32>
     margin=<ltrb 0 0 0 0>
     outlineSettings {
      bUseBrushTransparency=false
      roundingType=HalfHeightRadius
      width=0
      color {colorUseRule=UseColor_Specified, specifiedColor=<rgba 0 0 0 0>}
      cornerRadii=<0 0 0 0>
     }
     tintColor {colorUseRule=UseColor_Specified, specifiedColor=<rgba 1 1 1 1>}
     uVRegion {bIsValid=false, max=<0 0>, min=<0 0>}
    }
    renderTransform {angle=0, scale=<0.8 0.8>, shear=<0 0>, translation=<73 47>}
   }
   slot {class=OverlaySlot, properties {horizontalAlignment=HAlign_Fill, verticalAlignment=VAlign_Fill}}
  }
  Image_ChargeUp_Lightning {
   widgetClassPath=&Image
   bIsVariable=true
   parent=Reticle
   widget=Image_ChargeUp_Lightning
   properties {
    renderOpacity=0
    brush {
     bIsDynamicallyLoaded=false
     drawAs=Image
     imageType=FullColor
     mirroring=NoMirror
     resourceName=None
     resourceObject=/ChapWormGameplay/Items/ChapWorm/UI/Materials/MI_UI_ChapWorm_PwrUp_Lightning
     tiling=NoTile
     imageSize=<32 32>
     margin=<ltrb 0 0 0 0>
     outlineSettings {
      bUseBrushTransparency=false
      roundingType=HalfHeightRadius
      width=0
      color {colorUseRule=UseColor_Specified, specifiedColor=<rgba 0 0 0 0>}
      cornerRadii=<0 0 0 0>
     }
     tintColor {colorUseRule=UseColor_Specified, specifiedColor=<rgba 1 1 1 1>}
     uVRegion {bIsValid=false, max=<0 0>, min=<0 0>}
    }
    renderTransform {angle=0, scale=<0.7 0.7>, shear=<0 0>, translation=<73 47>}
   }
   slot {class=OverlaySlot, properties {horizontalAlignment=HAlign_Fill, verticalAlignment=VAlign_Fill}}
  }
  Image_Glow {
   widgetClassPath=&Image
   bIsVariable=true
   parent=Reticle
   widget=Image_Glow
   properties {
    renderOpacity=0
    brush {
     bIsDynamicallyLoaded=false
     drawAs=Image
     imageType=FullColor
     mirroring=NoMirror
     resourceName=None
     resourceObject=/ChapWormGameplay/Items/ChapWorm/UI/Materials/TEMP_MI_UI_ChapWorm_Sprint_Glow
     tiling=NoTile
     imageSize=<20 95>
     margin=<ltrb 0 0 0 0>
     outlineSettings {
      bUseBrushTransparency=false
      roundingType=HalfHeightRadius
      width=0
      color {colorUseRule=UseColor_Specified, specifiedColor=<rgba 0 0 0 0>}
      cornerRadii=<0 0 0 0>
     }
     tintColor {colorUseRule=UseColor_Specified, specifiedColor=<rgba 1 1 1 1>}
     uVRegion {bIsValid=false, max=<0 0>, min=<0 0>}
    }
    renderTransform {angle=35, scale=<1 1>, shear=<0 0>, translation=<76 45>}
   }
   slot {class=OverlaySlot, properties {horizontalAlignment=HAlign_Center, verticalAlignment=VAlign_Center}}
  }
  TutorialHighlight {widgetClassPath=&TutorialHighlight}
 ]
}
